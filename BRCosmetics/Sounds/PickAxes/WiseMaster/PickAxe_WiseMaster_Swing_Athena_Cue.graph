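# Sound cue for the WiseMaster cosmetic pickaxe swing, mixing a close layer with a medium-range layer.
# Each layer has its own attenuation and modulator, playing WiseMaster swing waves up close and shared Fire Axe distant swings at medium range.
# Quality level nodes shrink each random pool from six waves to three, then to a single wave at the lowest tiers.
# An owner-limited concurrency override stops the oldest swing when the limit is reached.
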SoundCue /BRCosmetics/Sounds/PickAxes/WiseMaster/PickAxe_WiseMaster_Swing_Athena_Cue
root=Mixer_1
nodes [
 {id=Attenuation_2, children [Modulator_3], properties {attenuationSettings=/Game/Sounds/Attenuation/Fort_Weapon_Pickaxe_Close_Att}}
 {id=Attenuation_3, children [Modulator_2], properties {attenuationSettings=/Game/Sounds/Attenuation/Fort_Weapon_Pickaxe_Medium_Att}}
 {id=Mixer_1, children [Attenuation_2, Attenuation_3], properties {inputVolume [1,1]}}
 {id=Modulator_2, children [QualityLevel_3]}
 {id=Modulator_3, children [QualityLevel_2]}
 {id=QualityLevel_2, children [Random_1, Random_4, WavePlayer_25, WavePlayer_25]}
 {id=QualityLevel_3, children [Random_3, Random_5, WavePlayer_29, WavePlayer_29]}
 {
  id=Random_1
  children [WavePlayer_10, WavePlayer_11, WavePlayer_12, WavePlayer_14, WavePlayer_15, WavePlayer_17]
  properties {weights [1,1,1,1,1,1]}
 }
 {
  id=Random_3
  children [WavePlayer_13, WavePlayer_16, WavePlayer_18, WavePlayer_19, WavePlayer_20, WavePlayer_21]
  properties {weights [1,1,1,1,1,1]}
 }
 {id=Random_4, children [WavePlayer_22, WavePlayer_23, WavePlayer_24], properties {weights [1,1,1]}}
 {id=Random_5, children [WavePlayer_27, WavePlayer_26, WavePlayer_28], properties {weights [1,1,1]}}
 {id=WavePlayer_10, properties {soundWaveAssetPtr=&/WiseMaster_Swing_01}}
 {id=WavePlayer_11, properties {soundWaveAssetPtr=&/WiseMaster_Swing_02}}
 {id=WavePlayer_12, properties {soundWaveAssetPtr=&/WiseMaster_Swing_03}}
 {id=WavePlayer_13, properties {soundWaveAssetPtr=/Game/Athena/Sounds/Weapons/Fire_Axe_Swing_Distant_04}}
 {id=WavePlayer_14, properties {soundWaveAssetPtr=&/WiseMaster_Swing_04}}
 {id=WavePlayer_15, properties {soundWaveAssetPtr=&/WiseMaster_Swing_05}}
 {id=WavePlayer_16, properties {soundWaveAssetPtr=/Game/Athena/Sounds/Weapons/Fire_Axe_Swing_Distant_01}}
 {id=WavePlayer_17, properties {soundWaveAssetPtr=&/WiseMaster_Swing_06}}
 {id=WavePlayer_18, properties {soundWaveAssetPtr=/Game/Athena/Sounds/Weapons/Fire_Axe_Swing_Distant_02}}
 {id=WavePlayer_19, properties {soundWaveAssetPtr=/Game/Athena/Sounds/Weapons/Fire_Axe_Swing_Distant_03}}
 {id=WavePlayer_20, properties {soundWaveAssetPtr=/Game/Athena/Sounds/Weapons/Fire_Axe_Swing_Distant_05}}
 {id=WavePlayer_21, properties {soundWaveAssetPtr=/Game/Athena/Sounds/Weapons/Fire_Axe_Swing_Distant_06}}
 {id=WavePlayer_22, properties {soundWaveAssetPtr=&WiseMaster_Swing_01}}
 {id=WavePlayer_23, properties {soundWaveAssetPtr=&WiseMaster_Swing_02}}
 {id=WavePlayer_24, properties {soundWaveAssetPtr=&WiseMaster_Swing_03}}
 {id=WavePlayer_25, properties {soundWaveAssetPtr=&WiseMaster_Swing_01}}
 {id=WavePlayer_26, properties {soundWaveAssetPtr=&Fire_Axe_Swing_Distant_03}}
 {id=WavePlayer_27, properties {soundWaveAssetPtr=&Fire_Axe_Swing_Distant_04}}
 {id=WavePlayer_28, properties {soundWaveAssetPtr=&Fire_Axe_Swing_Distant_06}}
 {id=WavePlayer_29, properties {soundWaveAssetPtr=&Fire_Axe_Swing_Distant_06}}
]
properties {
 audioPropertiesSheet=/BRCosmetics/Sounds/Settings/PropertySheets/Pickaxe/APS_Cosmetic_Pickaxe_Swing
 bOverrideConcurrency=true
 bPrimeOnLoad=true
 maxDistance=3000
 soundClassObject=&Fort_WeaponMelee_PSM
 volumeMultiplier=0.3
 concurrencyOverrides {bLimitToOwner=true, maxCount=2, resolutionRule=StopOldest}
 concurrencySet [&Player_Melee_Weapon_Conc]
}
